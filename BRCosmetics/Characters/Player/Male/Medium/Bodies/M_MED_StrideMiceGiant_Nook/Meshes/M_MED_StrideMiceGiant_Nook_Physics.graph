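# Physics asset for the StrideMiceGiant_Nook male medium body, covering the shirt's dynamic bones, pelvis and thighs.
# Five dyn_shirt capsules (back, sides, two fronts) hang from tiny base_shirt anchors with limited swing and zero twist.
# Anchors mostly chain through base_shirt to the pelvis with free angular motion, as do the larger thigh capsules.
# The left side breaks the pattern: dyn_shirt_l constrains directly to base_shirt, with base_shirt_l linked to it.
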PhysicsAsset /BRCosmetics/Characters/Player/Male/Medium/Bodies/M_MED_StrideMiceGiant_Nook/Meshes/M_MED_StrideMiceGiant_Nook_Physics
bodies {
 base_shirt [{shapeType=Capsule, length=0.505, radius=0.505, shapeName=base_shirt_capsule}]
 base_shirt_back [{shapeType=Capsule, length=0.505, radius=0.505, shapeName=base_shirt_back_capsule}]
 base_shirt_front_l [{shapeType=Capsule, length=0.505, radius=0.505, shapeName=base_shirt_front_l_capsule}]
 base_shirt_front_r [{shapeType=Capsule, length=0.505, radius=0.505, shapeName=base_shirt_front_r_capsule}]
 base_shirt_l [{shapeType=Capsule, length=0.505, radius=0.505, shapeName=base_shirt_l_capsule}]
 base_shirt_r [{shapeType=Capsule, length=0.505, radius=0.505, shapeName=base_shirt_r_capsule}]
 dyn_shirt_back [{shapeType=Capsule, length=6, radius=6, shapeName=dyn_shirt_back_capsule, center=<8 0 0>, rotation=<rot 0 0 90>}]
 dyn_shirt_front_l [{shapeType=Capsule, length=9, radius=6, shapeName=dyn_shirt_front_l_capsule, center=<8 0 3.3156>, rotation=<rot 0 0 90>}]
 dyn_shirt_front_r [{shapeType=Capsule, length=9, radius=6, shapeName=dyn_shirt_front_r_capsule, center=<8 0 4.1625>, rotation=<rot 0 0 90>}]
 dyn_shirt_l [{shapeType=Capsule, length=6, radius=6, shapeName=dyn_shirt_l_capsule, center=<8 0 -4.278>, rotation=<rot 0 0 90>}]
 dyn_shirt_r [{shapeType=Capsule, length=6, radius=6, shapeName=dyn_shirt_r_capsule, center=<8 0 4.7346>, rotation=<rot 0 0 90>}]
 pelvis [{shapeType=Capsule, length=11.33, radius=8.4919, shapeName=pelvis_capsule, rotation=<rot 0 -90 0>}]
 thigh_l [{shapeType=Capsule, length=15, radius=8, shapeName=thigh_l_capsule, center=<-18.3295 0 0>, rotation=<rot 90 0 0>}]
 thigh_r [{shapeType=Capsule, length=15, radius=7.5, shapeName=thigh_r_capsule, center=<16.56 0 0>, rotation=<rot 90 0 0>}]
}
constraints [
 {bone1Name=dyn_shirt_r, bone2Name=base_shirt_r, swing1LimitDegrees=5, swing2LimitDegrees=45, twistLimitDegrees=0}
 {
  bone1Name=base_shirt_l
  bone2Name=dyn_shirt_l
  swing1LimitDegrees=45
  swing1Motion=Free
  swing2LimitDegrees=45
  swing2Motion=Free
  twistLimitDegrees=45
  twistMotion=Free
 }
 {bone1Name=dyn_shirt_front_r, bone2Name=base_shirt_front_r, swing1LimitDegrees=20, swing2LimitDegrees=45, twistLimitDegrees=0}
 {bone1Name=dyn_shirt_front_l, bone2Name=base_shirt_front_l, swing1LimitDegrees=20, swing2LimitDegrees=45, twistLimitDegrees=0}
 {bone1Name=dyn_shirt_back, bone2Name=base_shirt_back, swing1LimitDegrees=5, swing2LimitDegrees=45, twistLimitDegrees=0}
 {
  bone1Name=base_shirt_front_r
  bone2Name=base_shirt
  swing1LimitDegrees=45
  swing1Motion=Free
  swing2LimitDegrees=45
  swing2Motion=Free
  twistLimitDegrees=45
  twistMotion=Free
 }
 {
  bone1Name=base_shirt_front_l
  bone2Name=base_shirt
  swing1LimitDegrees=45
  swing1Motion=Free
  swing2LimitDegrees=45
  swing2Motion=Free
  twistLimitDegrees=45
  twistMotion=Free
 }
 {
  bone1Name=base_shirt_back
  bone2Name=base_shirt
  swing1LimitDegrees=45
  swing1Motion=Free
  swing2LimitDegrees=45
  swing2Motion=Free
  twistLimitDegrees=45
  twistMotion=Free
 }
 {bone1Name=dyn_shirt_l, bone2Name=base_shirt, swing1LimitDegrees=5, swing2LimitDegrees=45, twistLimitDegrees=0}
 {
  bone1Name=base_shirt_r
  bone2Name=base_shirt
  swing1LimitDegrees=45
  swing1Motion=Free
  swing2LimitDegrees=45
  swing2Motion=Free
  twistLimitDegrees=45
  twistMotion=Free
 }
 {
  bone1Name=base_shirt
  bone2Name=pelvis
  swing1LimitDegrees=45
  swing1Motion=Free
  swing2LimitDegrees=45
  swing2Motion=Free
  twistLimitDegrees=45
  twistMotion=Free
 }
 {
  bone1Name=thigh_l
  bone2Name=pelvis
  swing1LimitDegrees=45
  swing1Motion=Free
  swing2LimitDegrees=45
  swing2Motion=Free
  twistLimitDegrees=45
  twistMotion=Free
 }
 {
  bone1Name=thigh_r
  bone2Name=pelvis
  swing1LimitDegrees=45
  swing1Motion=Free
  swing2LimitDegrees=45
  swing2Motion=Free
  twistLimitDegrees=45
  twistMotion=Free
 }
]
